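# Sound cue for the SiennaMale glider cosmetic's electrical one-shot layer, played as close-range glider detail.
# A Random node picks among twenty equally weighted Elec wave variations and feeds a Modulator that randomizes volume at reduced levels.
# The cue raises pitch and uses a glider close-detail attenuation plus its own concurrency set to limit overlapping instances.
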SoundCue /BRCosmetics/Sounds/Gliders/SiennaMale/Glider_SiennaMale_Elec_Cue
root=Modulator_0
nodes [
 {id=Modulator_0, children [Random_1], properties {volumeMax=0.6, volumeMin=0.25}}
 {
  id=Random_1
  children [
   WavePlayer_0
   WavePlayer_1
   WavePlayer_2
   WavePlayer_3
   WavePlayer_4
   WavePlayer_5
   WavePlayer_6
   WavePlayer_7
   WavePlayer_8
   WavePlayer_9
   WavePlayer_10
   WavePlayer_11
   WavePlayer_12
   WavePlayer_13
   WavePlayer_14
   WavePlayer_15
   WavePlayer_16
   WavePlayer_17
   WavePlayer_18
   WavePlayer_19
  ]
  properties {weights [1,1,1,1,1,1,1,1,1,1,1,1,1,1,1,1,1,1,1,1]}
 }
 {id=WavePlayer_0, properties {soundWaveAssetPtr=&/Glider_SiennaMale_Elec_01}}
 {id=WavePlayer_1, properties {soundWaveAssetPtr=&/Glider_SiennaMale_Elec_02}}
 {id=WavePlayer_10, properties {soundWaveAssetPtr=&/Glider_SiennaMale_Elec_11}}
 {id=WavePlayer_11, properties {soundWaveAssetPtr=&/Glider_SiennaMale_Elec_12}}
 {id=WavePlayer_12, properties {soundWaveAssetPtr=&/Glider_SiennaMale_Elec_13}}
 {id=WavePlayer_13, properties {soundWaveAssetPtr=&/Glider_SiennaMale_Elec_14}}
 {id=WavePlayer_14, properties {soundWaveAssetPtr=&/Glider_SiennaMale_Elec_15}}
 {id=WavePlayer_15, properties {soundWaveAssetPtr=&/Glider_SiennaMale_Elec_16}}
 {id=WavePlayer_16, properties {soundWaveAssetPtr=&/Glider_SiennaMale_Elec_17}}
 {id=WavePlayer_17, properties {soundWaveAssetPtr=&/Glider_SiennaMale_Elec_18}}
 {id=WavePlayer_18, properties {soundWaveAssetPtr=&/Glider_SiennaMale_Elec_19}}
 {id=WavePlayer_19, properties {soundWaveAssetPtr=&/Glider_SiennaMale_Elec_20}}
 {id=WavePlayer_2, properties {soundWaveAssetPtr=&/Glider_SiennaMale_Elec_03}}
 {id=WavePlayer_3, properties {soundWaveAssetPtr=&/Glider_SiennaMale_Elec_04}}
 {id=WavePlayer_4, properties {soundWaveAssetPtr=&/Glider_SiennaMale_Elec_05}}
 {id=WavePlayer_5, properties {soundWaveAssetPtr=&/Glider_SiennaMale_Elec_06}}
 {id=WavePlayer_6, properties {soundWaveAssetPtr=&/Glider_SiennaMale_Elec_07}}
 {id=WavePlayer_7, properties {soundWaveAssetPtr=&/Glider_SiennaMale_Elec_08}}
 {id=WavePlayer_8, properties {soundWaveAssetPtr=&/Glider_SiennaMale_Elec_09}}
 {id=WavePlayer_9, properties {soundWaveAssetPtr=&/Glider_SiennaMale_Elec_10}}
]
properties {
 attenuationSettings=/Game/Sounds/Attenuation/Fort_Player_Glider_CloseDetail_Athena_Att
 maxDistance=6000
 pitchMultiplier=1.4
 priority=0.5
 volumeMultiplier=1
 concurrencySet [&/Glider_SiennaMale_Elec_Concurrency]
}
